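# Physics asset for the F_MED_LimettaMech female player body cosmetic, providing secondary motion for a strap and four jacket flaps.
# Capsules on pelvis, spine_03 and both thighs form the body volume, while tiny base capsules anchor the dynamic chains.
# A four-link strap chain hangs off the pelvis, and single-bone front and back flaps hang off spine_03.
# Dynamic links have tight swing and twist limit angles; anchor and body constraints leave angular motion free.
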PhysicsAsset /BRCosmetics/Characters/Player/Female/Medium/Bodies/F_MED_LimettaMech/Meshes/F_MED_LimettaMech_Physics
bodies {
 base_Strap [{shapeType=Capsule, length=0.505, radius=0.505, shapeName=base_Strap_capsule, rotation=<rot 0 -90 0>}]
 base_back_Jacket_l [{shapeType=Capsule, length=0.505, radius=0.505, shapeName=base_back_Jacket_l_capsule, rotation=<rot 0 -90 0>}]
 base_back_Jacket_r [{shapeType=Capsule, length=0.505, radius=0.505, shapeName=base_back_Jacket_r_capsule, rotation=<rot 0 -90 0>}]
 base_front_Jacket_l [{shapeType=Capsule, length=0.505, radius=0.505, shapeName=base_front_Jacket_l_capsule, rotation=<rot 0 -90 0>}]
 base_front_Jacket_r [{shapeType=Capsule, length=0.505, radius=0.505, shapeName=base_front_Jacket_r_capsule, rotation=<rot 0 -90 0>}]
 dyn_Strap_01 [{shapeType=Capsule, length=2, radius=0.8774, shapeName=dyn_Strap_01_capsule, center=<2 0 0>, rotation=<rot 90 0 0>}]
 dyn_Strap_02 [{shapeType=Capsule, length=2.4616, radius=1.1092, shapeName=dyn_Strap_02_capsule, center=<3.4171 -0.1359 0>, rotation=<rot 90 180 -177.0814>}]
 dyn_Strap_03 [{shapeType=Capsule, length=2.2616, radius=1.2, shapeName=dyn_Strap_03_capsule, center=<1.4488 0 0>, rotation=<rot 90 0 0>}]
 dyn_Strap_04 [{shapeType=Capsule, length=2.8902, radius=1.3, shapeName=dyn_Strap_04_capsule, center=<2.5349 0 -0.2268>, rotation=<rot 90 180 175.1175>}]
 dyn_back_Jacket_l [{shapeType=Capsule, length=6, radius=2.5, shapeName=dyn_back_Jacket_l_capsule, center=<2 0 0>, rotation=<rot 85 180 180>}]
 dyn_back_Jacket_r [{shapeType=Capsule, length=6, radius=2.5, shapeName=dyn_back_Jacket_r_capsule, center=<-2 0 0>, rotation=<rot -85 180 0>}]
 dyn_front_Jacket_l [{shapeType=Capsule, length=6, radius=2.5, shapeName=dyn_front_Jacket_l_capsule, center=<4 0 0>, rotation=<rot 85 0 0>}]
 dyn_front_Jacket_r [{shapeType=Capsule, length=6, radius=2.5, shapeName=dyn_front_Jacket_r_capsule, center=<-4 0 0>, rotation=<rot 88.1536 -180 -180>}]
 pelvis [{
   shapeType=Capsule
   length=15.0367
   radius=9.312
   shapeName=pelvis_capsule
   center=<-5.1355 -3.0409 -0.0058>
   rotation=<rot -2.0617 -90 -0.9499>
  }]
 spine_03 [{
   shapeType=Capsule
   length=11.3225
   radius=8.9839
   shapeName=spine_03_capsule
   center=<5.7768 1.1289 0.0443>
   rotation=<rot 0 -90.0001 -90.4394>
  }]
 thigh_l [{
   shapeType=Capsule
   length=10
   radius=8.7532
   shapeName=thigh_l_capsule
   center=<-15.024 0.4984 0.6833>
   rotation=<rot 1.1826 -103.8727 84.0591>
  }]
 thigh_r [{
   shapeType=Capsule
   length=8.407
   radius=8.2822
   shapeName=thigh_r_capsule
   center=<13.6527 -0.1508 -0.261>
   rotation=<rot 1.1826 -103.8727 -91.807>
  }]
}
constraints [
 {
  bone1Name=base_Strap
  bone2Name=pelvis
  swing1LimitDegrees=45
  swing1Motion=Free
  swing2LimitDegrees=45
  swing2Motion=Free
  twistLimitDegrees=45
  twistMotion=Free
 }
 {bone1Name=dyn_Strap_01, bone2Name=base_Strap, swing1LimitDegrees=4, swing2LimitDegrees=2, twistLimitDegrees=1}
 {bone1Name=dyn_Strap_02, bone2Name=dyn_Strap_01, swing1LimitDegrees=4, swing2LimitDegrees=2, twistLimitDegrees=1}
 {bone1Name=dyn_Strap_03, bone2Name=dyn_Strap_02, swing1LimitDegrees=5, swing2LimitDegrees=3, twistLimitDegrees=2}
 {bone1Name=dyn_Strap_04, bone2Name=dyn_Strap_03, swing1LimitDegrees=10, swing2LimitDegrees=5, twistLimitDegrees=1}
 {
  bone1Name=thigh_r
  bone2Name=pelvis
  swing1LimitDegrees=45
  swing1Motion=Free
  swing2LimitDegrees=45
  swing2Motion=Free
  twistLimitDegrees=45
  twistMotion=Free
 }
 {
  bone1Name=base_back_Jacket_l
  bone2Name=spine_03
  swing1LimitDegrees=45
  swing1Motion=Free
  swing2LimitDegrees=45
  swing2Motion=Free
  twistLimitDegrees=45
  twistMotion=Free
 }
 {
  bone1Name=base_back_Jacket_r
  bone2Name=spine_03
  swing1LimitDegrees=45
  swing1Motion=Free
  swing2LimitDegrees=45
  swing2Motion=Free
  twistLimitDegrees=45
  twistMotion=Free
 }
 {
  bone1Name=base_front_Jacket_l
  bone2Name=spine_03
  swing1LimitDegrees=45
  swing1Motion=Free
  swing2LimitDegrees=45
  swing2Motion=Free
  twistLimitDegrees=45
  twistMotion=Free
 }
 {
  bone1Name=base_front_Jacket_r
  bone2Name=spine_03
  swing1LimitDegrees=45
  swing1Motion=Free
  swing2LimitDegrees=45
  swing2Motion=Free
  twistLimitDegrees=45
  twistMotion=Free
 }
 {bone1Name=dyn_back_Jacket_l, bone2Name=base_back_Jacket_l, swing1LimitDegrees=0, swing2LimitDegrees=5, twistLimitDegrees=0}
 {bone1Name=dyn_back_Jacket_r, bone2Name=base_back_Jacket_r, swing1LimitDegrees=0, swing2LimitDegrees=5, twistLimitDegrees=0}
 {bone1Name=dyn_front_Jacket_l, bone2Name=base_front_Jacket_l, swing1LimitDegrees=1, swing2LimitDegrees=8, twistLimitDegrees=0.5}
 {bone1Name=dyn_front_Jacket_r, bone2Name=base_front_Jacket_r, swing1LimitDegrees=1, swing2LimitDegrees=8, twistLimitDegrees=0.5}
 {
  bone1Name=spine_03
  bone2Name=pelvis
  swing1LimitDegrees=45
  swing1Motion=Free
  swing2LimitDegrees=45
  swing2Motion=Free
  twistLimitDegrees=45
  twistMotion=Free
 }
 {
  bone1Name=thigh_l
  bone2Name=pelvis
  swing1LimitDegrees=45
  swing1Motion=Free
  swing2LimitDegrees=45
  swing2Motion=Free
  twistLimitDegrees=45
  twistMotion=Free
 }
]
