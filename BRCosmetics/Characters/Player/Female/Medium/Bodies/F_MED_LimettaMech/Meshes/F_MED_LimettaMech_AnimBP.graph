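# Post-process style AnimBP for the LimettaMech female body cosmetic, layering secondary motion onto the parent mesh's copied pose.
# A Control Rig corrects the pose, then AnimDynamics, RigidBody and a zipper constraint drive clothing and straps.
# Grenade dynamics switch between world and root-relative simulation via bIsInFrontEnd, before fixed belt and forearm offsets.
# The event graph updates RigidBody sim-space settings by movement state, plus targeting, melee and gauntlet inputs for the rig.
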AnimBlueprint /BRCosmetics/Characters/Player/Female/Medium/Bodies/F_MED_LimettaMech/Meshes/F_MED_LimettaMech_AnimBP
target_skeleton=&/F_MED_LimettaMech_Skeleton
graphs [
 AnimGraph {nodes [
   Output Pose(Result=Use cached pose 'CR_cache_offset'()) {layerGroup=None}
   n1 = Control Rig!(
    Source=Copy Pose From Mesh() {bUseAttachedParent=true}
    Alpha
    Is Crouching:bool=Get bIsCrouching()
    IsTargeting:double=Get isTargeting float() -> (isTargeting float:double)
    DBNO:bool=Get bIsDBNO()
    Playing Emote:bool=Get bIsPlayingEmote()
    IsWearingGloves:bool=Get IsWearingGauntlet() -> (IsWearingGauntlet:bool)
    MeeLeeWeaponOn:double=Get IsMeeleeEquipped() -> (IsMeeleeEquipped:double)
   ) -> (Pose) {
    alphaBoolBlend {blendOption=Linear}
    controlRigAssetReference {blueprintRigClass=&/F_MED_LimettaMech_CtrlRig.F_MED_LimettaMech_CtrlRig_C}
   }
   Save cached pose 'CR_cache_Input'(Pose=@n1) {cacheName=CR_cache_Input}
   n2 = "Transform (Modify) Bone - Bone: dyn_belt_1"!(ComponentPose=Local To Component(LocalPose=Use cached pose 'CRCache_Animdynamics'()), bAlphaBoolEnabled:bool=True, Translation="0,3,2") -> (Pose) {
    alphaInputType=Bool
    rotationMode=BMM_Additive
    translationMode=BMM_Additive
    alphaBoolBlend {blendInTime=0.1, blendOption=Linear, blendOutTime=0.1}
    boneToModify {boneName=dyn_belt_1}
    translation {y=3, z=2}
   }
   n3 = "Transform (Modify) Bone - Bone: lowerarm_r"!(ComponentPose=@n2, bAlphaBoolEnabled:bool=True, Scale="0.5,0.5,0.5") -> (Pose) {
    alphaInputType=Bool
    scaleMode=BMM_Additive
    scaleSpace=BCS_BoneSpace
    alphaBoolBlend {blendOption=Linear}
    boneToModify {boneName=lowerarm_r}
    scale=<0.5 0.5 0.5>
   }
   n4 = "Transform (Modify) Bone - Bone: lowerarm_l"!(ComponentPose=@n3, bAlphaBoolEnabled:bool=True, Scale="0.5,0.5,0.5") -> (Pose) {
    alphaInputType=Bool
    scaleMode=BMM_Additive
    scaleSpace=BCS_BoneSpace
    alphaBoolBlend {blendOption=Linear}
    boneToModify {boneName=lowerarm_l}
    scale=<0.5 0.5 0.5>
   }
   Save cached pose 'CR_cache_offset'(Pose=Component To Local(ComponentPose=@n4)) {cacheName=CR_cache_offset}
   n5 = Anim Dynamics - Bone: dyn_nade_a_1!(ComponentPose=Local To Component(LocalPose=Use cached pose 'CRCache_dynamics'()), Alpha=1, GravityOverride:Vector="0,0,0") -> (Pose) {
    angularSpringConstant=12000
    bAngularSpring=true
    bUseGravityOverride=true
    simulationSpace=RootRelative
    alphaBoolBlend {blendOption=Linear}
    boundBone {boneName=dyn_nade_a_1}
    componentAppliedLinearAccClamp=<3000 3000 3000>
    componentLinearAccScale=<0.2 0.2 0.2>
    componentLinearVelScale=<0.2 0.2 0.2>
    physicsBodyDefinitions [{
      boundBone {boneName=dyn_nade_a_1}
      boxExtents=<15 15 15>
      constraintSetup {angularLimitsMax=<0 0 4>, angularLimitsMin=<0 -6 -4>, angularTarget=<0 0 0>}
      localJointOffset=<2 0 0>
     }]
   }
   n6 = Anim Dynamics - Bone: dyn_nade_handle_01!(ComponentPose=@n5, Alpha=1, GravityOverride:Vector="0,0,-980") -> (Pose) {
    angularDampingOverride=1
    angularSpringConstant=20000
    bAngularSpring=true
    bOverrideAngularDamping=true
    bUseGravityOverride=true
    gravityScale=0.35
    simulationSpace=RootRelative
    alphaBoolBlend {blendOption=Linear}
    boundBone {boneName=dyn_nade_handle_01}
    componentLinearAccScale=<1 1 1>
    componentLinearVelScale=<1 1 1>
    gravityOverride {z=-980}
    physicsBodyDefinitions [{
      boundBone {boneName=dyn_nade_handle_01}
      boxExtents=<20 30 30>
      constraintSetup {angularLimitsMax=<7 7 7>, angularLimitsMin=<-7 -7 -7>, angularTarget=<1 0 0>}
      localJointOffset=<4 0 0>
     }]
   }
   n7 = Anim Dynamics - Bone: dyn_nade_a_2!(ComponentPose=@n6, Alpha=1, GravityOverride:Vector="0,0,0") -> (Pose) {
    angularSpringConstant=12000
    bAngularSpring=true
    bUseGravityOverride=true
    simulationSpace=RootRelative
    alphaBoolBlend {blendOption=Linear}
    boundBone {boneName=dyn_nade_a_2}
    componentAppliedLinearAccClamp=<3000 3000 3000>
    componentLinearAccScale=<0.2 0.2 0.2>
    componentLinearVelScale=<0.2 0.2 0.2>
    physicsBodyDefinitions [{
      boundBone {boneName=dyn_nade_a_2}
      boxExtents=<15 15 15>
      constraintSetup {angularLimitsMax=<0 0 4>, angularLimitsMin=<0 -6 -4>, angularTarget=<0 0 0>}
      localJointOffset=<2 0 0>
     }]
   }
   n8 = Anim Dynamics - Bone: dyn_nade_handle_02!(ComponentPose=@n7, Alpha=1, GravityOverride:Vector="0,0,-980") -> (Pose) {
    angularDampingOverride=1
    angularSpringConstant=20000
    bAngularSpring=true
    bOverrideAngularDamping=true
    bUseGravityOverride=true
    gravityScale=0.35
    simulationSpace=RootRelative
    alphaBoolBlend {blendOption=Linear}
    boundBone {boneName=dyn_nade_handle_02}
    componentLinearAccScale=<1 1 1>
    componentLinearVelScale=<1 1 1>
    gravityOverride {z=-980}
    physicsBodyDefinitions [{
      boundBone {boneName=dyn_nade_handle_02}
      boxExtents=<20 30 30>
      constraintSetup {angularLimitsMax=<7 7 7>, angularLimitsMin=<-7 -7 -7>, angularTarget=<1 0 0>}
      localJointOffset=<4 0 0>
     }]
   }
   n9 = Anim Dynamics - Bone: dyn_nade_a_1(ComponentPose=Local To Component(LocalPose=Use cached pose 'CRCache_dynamics'()), Alpha=0.7, GravityOverride="0,0,0") {
    angularSpringConstant=12000
    bAngularSpring=true
    bUseGravityOverride=true
    simulationSpace=World
    alphaBoolBlend {blendOption=Linear}
    boundBone {boneName=dyn_nade_a_1}
    componentAppliedLinearAccClamp=<3000 3000 3000>
    componentLinearAccScale=<0.2 0.2 0.2>
    componentLinearVelScale=<0.2 0.2 0.2>
    physicsBodyDefinitions [{
      boundBone {boneName=dyn_nade_a_1}
      boxExtents=<8 8 8>
      constraintSetup {angularLimitsMax=<0 0 4>, angularLimitsMin=<0 -6 -4>, angularTarget=<0 0 0>}
      localJointOffset=<1 0 0>
     }]
   }
   n10 = Anim Dynamics - Bone: dyn_nade_handle_01(ComponentPose=@n9, Alpha=0.7, GravityOverride="0,0,-980") {
    angularDampingOverride=1
    angularSpringConstant=20000
    bAngularSpring=true
    bOverrideAngularDamping=true
    bUseGravityOverride=true
    gravityScale=0.35
    simulationSpace=World
    alphaBoolBlend {blendOption=Linear}
    boundBone {boneName=dyn_nade_handle_01}
    componentLinearAccScale=<1 1 1>
    componentLinearVelScale=<1 1 1>
    gravityOverride {z=-980}
    physicsBodyDefinitions [{
      boundBone {boneName=dyn_nade_handle_01}
      boxExtents=<20 30 30>
      constraintSetup {angularLimitsMax=<7 7 7>, angularLimitsMin=<-7 -7 -7>, angularTarget=<1 0 0>}
      localJointOffset=<4 0 0>
     }]
   }
   n11 = Anim Dynamics - Bone: dyn_nade_a_2(ComponentPose=@n10, Alpha=0.7, GravityOverride="0,0,0") {
    angularSpringConstant=12000
    bAngularSpring=true
    bUseGravityOverride=true
    simulationSpace=World
    alphaBoolBlend {blendOption=Linear}
    boundBone {boneName=dyn_nade_a_2}
    componentAppliedLinearAccClamp=<3000 3000 3000>
    componentLinearAccScale=<0.2 0.2 0.2>
    componentLinearVelScale=<0.2 0.2 0.2>
    physicsBodyDefinitions [{
      boundBone {boneName=dyn_nade_a_2}
      boxExtents=<8 8 8>
      constraintSetup {angularLimitsMax=<0 0 4>, angularLimitsMin=<0 -6 -4>, angularTarget=<0 0 0>}
      localJointOffset=<1 0 0>
     }]
   }
   n12 = Anim Dynamics - Bone: dyn_nade_handle_02(ComponentPose=@n11, Alpha=0.7, GravityOverride="0,0,-980") {
    angularDampingOverride=1
    angularSpringConstant=20000
    bAngularSpring=true
    bOverrideAngularDamping=true
    bUseGravityOverride=true
    gravityScale=0.35
    simulationSpace=World
    alphaBoolBlend {blendOption=Linear}
    boundBone {boneName=dyn_nade_handle_02}
    componentLinearAccScale=<1 1 1>
    componentLinearVelScale=<1 1 1>
    gravityOverride {z=-980}
    physicsBodyDefinitions [{
      boundBone {boneName=dyn_nade_handle_02}
      boxExtents=<20 30 30>
      constraintSetup {angularLimitsMax=<7 7 7>, angularLimitsMin=<-7 -7 -7>, angularTarget=<1 0 0>}
      localJointOffset=<4 0 0>
     }]
   }
   Save cached pose 'CRCache_Animdynamics'(Pose=Blend Poses by bool(BlendPose_0=Component To Local(ComponentPose=@n12), BlendPose_1=Component To Local(ComponentPose=@n8), bActiveValue=Get bIsInFrontEnd()) {blendType=Linear}) {cacheName=CRCache_Animdynamics}
   n13 = Anim Dynamics - Bone: dyn_collar_r_01(ComponentPose=Local To Component(LocalPose=Use cached pose 'CR_cache_Input'()), Alpha=1, ExternalForce="0,0,0") {
    angularDampingOverride=1
    angularSpringConstant=20000
    bAngularSpring=true
    bOverrideAngularDamping=true
    bUseGravityOverride=true
    gravityScale=0.35
    simulationSpace=RootRelative
    alphaBoolBlend {blendOption=Linear}
    boundBone {boneName=dyn_collar_r_01}
    componentLinearAccScale=<1 1 1>
    componentLinearVelScale=<1 1 1>
    gravityOverride {z=-980}
    physicsBodyDefinitions [{
      boundBone {boneName=dyn_collar_r_01}
      boxExtents=<20 30 30>
      constraintSetup {angularLimitsMax=<5 2.5 2.5>, angularLimitsMin=<-5 -2.5 -2.5>, angularTarget=<1 0 0>}
      localJointOffset=<2 0 0>
     }]
   }
   n14 = Anim Dynamics - Bone: dyn_collar_l_01(ComponentPose=@n13, Alpha=1, ExternalForce="0,0,0") {
    angularDampingOverride=1
    angularSpringConstant=20000
    bAngularSpring=true
    bOverrideAngularDamping=true
    bUseGravityOverride=true
    gravityScale=0.35
    simulationSpace=RootRelative
    alphaBoolBlend {blendOption=Linear}
    boundBone {boneName=dyn_collar_l_01}
    componentLinearAccScale=<1 1 1>
    componentLinearVelScale=<1 1 1>
    gravityOverride {z=-980}
    physicsBodyDefinitions [{
      boundBone {boneName=dyn_collar_l_01}
      boxExtents=<20 30 30>
      constraintSetup {angularLimitsMax=<5 2.5 2.5>, angularLimitsMin=<-5 -2.5 -2.5>, angularTarget=<1 0 0>}
      localJointOffset=<2 0 0>
     }]
   }
   n15 = Anim Dynamics - Bone: dyn_collar_l_01(ComponentPose=@n14, Alpha=1, ExternalForce="0,0,0") {
    angularDampingOverride=1
    angularSpringConstant=20000
    bAngularSpring=true
    bOverrideAngularDamping=true
    bUseGravityOverride=true
    gravityScale=0.35
    simulationSpace=RootRelative
    alphaBoolBlend {blendOption=Linear}
    boundBone {boneName=dyn_collar_l_01}
    componentLinearAccScale=<1 1 1>
    componentLinearVelScale=<1 1 1>
    gravityOverride {z=-980}
    physicsBodyDefinitions [{
      boundBone {boneName=dyn_collar_l_01}
      boxExtents=<20 30 30>
      constraintSetup {angularLimitsMax=<5 2.5 2.5>, angularLimitsMin=<-5 -2.5 -2.5>, angularTarget=<1 0 0>}
      localJointOffset=<2 0 0>
     }]
   }
   n16 = RigidBody!(
    ComponentPose=@n15
    Alpha
    OverrideWorldGravity:Vector="0,0,-980"
    SimSpaceSettings:Sim Space Settings Structure=Get Sim Space Settings() -> (Sim Space Settings:Sim Space Settings Structure)
   ) -> (Pose) {
    bClampLinearTranslationLimitToRefPose=true
    bOverrideWorldGravity=true
    overridePhysicsAsset=&/F_MED_LimettaMech_Physics
    simulationSpace=BaseBoneSpace
    alphaBoolBlend {blendOption=Linear}
    baseBoneRef {boneName=pelvis}
    componentAppliedLinearAccClamp=<5000 5000 5000>
    simSpaceSettings {
     maxAngularAcceleration=1000
     maxAngularVelocity=5
     maxLinearAcceleration=700
     maxLinearVelocity=700
     velocityScaleZ=0.5
     worldAlpha=1
     externalLinearDragV {x=0.2}
    }
   }
   n17 = "Constraint - base_jacket_Zipper (PP)"!(ComponentPose=@n16, Alpha=1, ConstraintWeights_0=0.999, ConstraintWeights_1:float=0.001) -> (Pose) {
    alphaBoolBlend {blendOption=Linear}
    boneToModify {boneName=base_jacket_Zipper}
    constraintSetup [{targetBone {boneName=dyn_front_Jacket_r}}, {targetBone {boneName=dyn_back_Jacket_r}}]
    constraintWeights [1,1]
   }
   n18 = Anim Dynamics - Bone: dyn_jacket_Zipper(ComponentPose=@n17, Alpha=1, ExternalForce="0,0,0") {
    angularDampingOverride=1
    angularSpringConstant=20000
    bOverrideAngularDamping=true
    bUseGravityOverride=true
    gravityScale=0.35
    simulationSpace=RootRelative
    alphaBoolBlend {blendOption=Linear}
    boundBone {boneName=dyn_jacket_Zipper}
    componentAppliedLinearAccClamp=<3000 3000 3000>
    componentLinearAccScale=<0.6 0.6 0.6>
    componentLinearVelScale=<0.6 0.6 0.6>
    gravityOverride {z=-980}
    physicsBodyDefinitions [{
      boundBone {boneName=dyn_jacket_Zipper}
      boxExtents=<10 16 8>
      constraintSetup {angularLimitsMax=<5 20 10>, angularLimitsMin=<-5 -70 -10>, angularTarget=<1 0 0>}
      localJointOffset=<4 0 0>
     }]
   }
   n19 = Anim Dynamics - Bone: dyn_bootStrap_r(ComponentPose=@n18, Alpha=1, ExternalForce="0,0,0") {
    angularDampingOverride=1
    angularSpringConstant=20000
    bAngularSpring=true
    bOverrideAngularDamping=true
    bUseGravityOverride=true
    gravityScale=0.35
    simulationSpace=RootRelative
    alphaBoolBlend {blendOption=Linear}
    boundBone {boneName=dyn_bootStrap_r}
    componentLinearAccScale=<1 1 1>
    componentLinearVelScale=<1 1 1>
    gravityOverride {z=-980}
    physicsBodyDefinitions [{
      boundBone {boneName=dyn_bootStrap_r}
      boxExtents=<20 30 30>
      constraintSetup {angularLimitsMax=<15 20 25>, angularLimitsMin=<-15 -20 -25>, angularTarget=<1 0 0>}
      localJointOffset=<4 0 0>
     }]
   }
   n20 = Anim Dynamics - Bone: dyn_bootStrap_l(ComponentPose=@n19, Alpha=1, ExternalForce="0,0,0") {
    angularDampingOverride=1
    angularSpringConstant=20000
    bAngularSpring=true
    bOverrideAngularDamping=true
    bUseGravityOverride=true
    gravityScale=0.35
    simulationSpace=RootRelative
    alphaBoolBlend {blendOption=Linear}
    boundBone {boneName=dyn_bootStrap_l}
    componentLinearAccScale=<1 1 1>
    componentLinearVelScale=<1 1 1>
    gravityOverride {z=-980}
    physicsBodyDefinitions [{
      boundBone {boneName=dyn_bootStrap_l}
      boxExtents=<20 30 30>
      constraintSetup {angularLimitsMax=<15 20 25>, angularLimitsMin=<-15 -20 -25>, angularTarget=<1 0 0>}
      localJointOffset=<4 0 0>
     }]
   }
   Save cached pose 'CRCache_dynamics'(Pose=Component To Local(ComponentPose=@n20)) {cacheName=CRCache_dynamics}
  ]}
 EventGraph {nodes [
   n1 = Event BlueprintUpdateAnimation()
   > n2 = Sequence()
   > n3 = Gravity Override State Function DEPRECATED(
    CharacterPartAnimInstance=Self-Reference()
    GameTimeInSeconds=GetGameTimeInSeconds()
    GravityOverrideParams=Get Gravity Override Params()
    RBAN Params=Get RBAN Params()
   )
   > Set GravityOverride(GravityOverride=@n3)
   > Set Sim Space Settings(Sim Space Settings:Sim Space Settings Structure=Break RBANSetting Struct!(RBANSettingStruct:"RBANSetting Struct Structure (by ref)"=@n3.RBAN SettingsOut) -> (
     ComponentLinearAccScale_28_#1:Vector=~
     ComponentLinearVelScale_31_#2:Vector=~
     SimSpaceSettings_34_#3:Sim Space Settings Structure
    ) {is=title K2Node_BreakStruct}.SimSpaceSettings_34_#3) -> (Output_Get:Sim Space Settings Structure)
   n4 = Get bIsTargeting()
   Set isTargeting float(
    execute=@n2.then_1
    isTargeting float:double=FInterpTo(Current=Get isTargeting float(), Target="To Float (Boolean)"(InBool=@n4), DeltaTime=@n1.DeltaTimeX, InterpSpeed=15)
   ) -> (Output_Get:double)
   n5 = Event OnCharacterAndPickaxeCustomizationLoaded!() -> (
    OutputDelegate
    then
    CustomBaseMesh:Skeletal Mesh Component Object Reference
    CustomParts:Array of Custom Character Part Object References
    CustomMeshParts:Array of Skeletal Mesh Component Budgeted Object References
   ) {is=title K2Node_Event}
   > n6 = Branch(Condition=IsNullItem!(ItemObject:Object Reference=Get Array Item with EFort Custom Part Type(Array:Array of Custom Character Part Object References=@n5.CustomParts, PartType=Gameplay) -> (Output:Custom Character Part Object Reference)) -> (Is Null Item:bool))
   > Set IsWearingGauntlet(IsWearingGauntlet:bool=false) -> (Output_Get:bool)
   Set IsWearingGauntlet(execute=@n6.else, IsWearingGauntlet=true)
   Event OnWeaponChanged()
   > Set isGunOut(isGunOut:double=1) -> (Output_Get:double)
   > n7 = Switch on EFortWeaponCoreAnimation!(execute, Selection:EFortWeaponCoreAnimation Enum=Get WeaponCoreAnim!() -> (WeaponCoreAnim)) -> (
    Melee:Exec
    Pistol:Exec
    Shotgun:Exec
    PaperBlueprint:Exec
    Rifle:Exec
    MeleeOneHand:Exec
    MachinePistol:Exec
    RocketLauncher:Exec
    GrenadeLauncher:Exec
    GoingCommando:Exec
    AssaultRifle:Exec
    TacticalShotgun:Exec
    SniperRifle:Exec
    TrapPlacement:Exec
    ShoulderLauncher:Exec
    AbilityDecoTool:Exec
    Crossbow:Exec
    C4:Exec
    RemoteControl:Exec
    DualWield:Exec
    AR_BullPup:Exec
    AR_ForwardGrip:Exec
    MedPackPaddles:Exec
    SMG_P90:Exec
    AR_DrumGun:Exec
    Consumable_Small:Exec
    Consumable_Large:Exec
    Balloon:Exec
    MountedTurret:Exec
    CreativeTool:Exec
    ExplosiveBow:Exec
    AshtonIndigo:Exec
    AshtonChicago:Exec
    MeleeDualWield:Exec
    Unarmed:Exec
   )
   n8 = Sequence(execute=[@n7.Melee, @n7.MeleeDualWield, @n7.MeleeOneHand])
   > Set IsMeeleeEquipped(IsMeeleeEquipped:double=1) -> (Output_Get:double)
   Set isGunOut(execute=[@n8.then_1, @n7.Unarmed], isGunOut=0)
  ]}
]
variables [
 fCrouching {type=boolean}
 Gravity Override Params {type=GravityOverrideParamsStruct}
 GravityOverride {type=Vector}
 isGunOut {type=number}
 IsMeeleeEquipped {type=number}
 isTargeting float {type=number}
 IsWearingGauntlet {type=boolean}
 RBAN Params {
  type=RBANParamsStruct
  default {
   dBNO {simSpaceSettings {
     maxAngularAcceleration=1000
     maxAngularVelocity=5
     maxLinearAcceleration=700
     maxLinearVelocity=700
     velocityScaleZ=0.5
     externalLinearDragV=<0.2 0.2 0.2>
    }}
   dBNO_Carried {simSpaceSettings {
     maxAngularAcceleration=1000
     maxAngularVelocity=5
     maxLinearAcceleration=700
     maxLinearVelocity=700
     velocityScaleZ=0.5
     externalLinearDragV=<0.2 0.2 0.2>
    }}
   emoteOrMelee {simSpaceSettings {
     maxAngularAcceleration=1000
     maxAngularVelocity=5
     maxLinearAcceleration=700
     maxLinearVelocity=700
     velocityScaleZ=0.5
     externalLinearDragV=<0.2 0.2 0.2>
    }}
   front_End {simSpaceSettings {
     maxAngularAcceleration=8
     maxAngularVelocity=2
     maxLinearAcceleration=40
     maxLinearVelocity=20
     velocityScaleZ=0
     externalLinearDragV=<0.1 0.1 0.1>
    }}
   onGround_Crouch_Moving {simSpaceSettings {
     maxAngularAcceleration=1000
     maxAngularVelocity=5
     maxLinearAcceleration=700
     maxLinearVelocity=700
     velocityScaleZ=0.5
     externalLinearDragV=<0.2 0.2 0.2>
    }}
   onGround_Falling {simSpaceSettings {
     maxAngularAcceleration=1000
     maxAngularVelocity=5
     maxLinearAcceleration=700
     maxLinearVelocity=700
     velocityScaleZ=0.5
     externalLinearDragV=<0.2 0.2 0.2>
    }}
   onGround_Moving {simSpaceSettings {
     maxAngularAcceleration=1000
     maxAngularVelocity=5
     maxLinearAcceleration=700
     maxLinearVelocity=700
     velocityScaleZ=0.5
     externalLinearDragV=<0.2 0 0>
    }}
   onGround_Standing {simSpaceSettings {
     maxAngularAcceleration=1000
     maxAngularVelocity=5
     maxLinearAcceleration=700
     maxLinearVelocity=700
     velocityScaleZ=0.5
     externalLinearDragV=<0.2 0 0>
    }}
   rail_Grinding {simSpaceSettings {
     maxAngularAcceleration=1000
     maxAngularVelocity=5
     maxLinearAcceleration=700
     maxLinearVelocity=700
     velocityScaleZ=0.5
     worldAlpha=0.1
     externalLinearDragV=<0.2 0 0>
    }}
   riding_Moving {simSpaceSettings {
     maxAngularAcceleration=1000
     maxAngularVelocity=5
     maxLinearAcceleration=700
     maxLinearVelocity=700
     velocityScaleZ=0.5
     externalLinearDragV=<0.2 0 0>
    }}
   riding_Moving_Targeting {simSpaceSettings {
     maxAngularAcceleration=1000
     maxAngularVelocity=5
     maxLinearAcceleration=700
     maxLinearVelocity=700
     velocityScaleZ=0.5
     externalLinearDragV=<0.2 0 0>
    }}
   riding_Standing {simSpaceSettings {
     maxAngularAcceleration=1000
     maxAngularVelocity=3
     maxLinearAcceleration=700
     maxLinearVelocity=700
     velocityScaleZ=1
     externalLinearDragV=<0.2 0.2 0.2>
    }}
   skydive_DivingDown {simSpaceSettings {
     maxAngularAcceleration=1000
     maxAngularVelocity=5
     maxLinearAcceleration=700
     maxLinearVelocity=700
     velocityScaleZ=0.5
     externalLinearDragV=<0.2 0 0>
    }}
   skydive_FloatingDown {
    componentLinearAccScale=<0 0.5 0>
    componentLinearVelScale=<0 0.5 0>
    simSpaceSettings {
     maxAngularAcceleration=1000
     maxAngularVelocity=5
     maxLinearAcceleration=700
     maxLinearVelocity=700
     velocityScaleZ=0.5
     externalLinearDragV=<0.2 0 0>
    }
   }
   skydive_GoingUp {simSpaceSettings {
     maxAngularAcceleration=1000
     maxAngularVelocity=5
     maxLinearAcceleration=700
     maxLinearVelocity=700
     velocityScaleZ=0.5
     externalLinearDragV=<0.2 0 0>
    }}
   skydive_Parachuting {
    componentLinearAccScale=<0 0.1 0.3>
    componentLinearVelScale=<0 0.1 0.3>
    simSpaceSettings {
     maxAngularAcceleration=1000
     maxAngularVelocity=5
     maxLinearAcceleration=700
     maxLinearVelocity=700
     velocityScaleZ=0.5
     externalLinearDragV=<0.2 0 0>
    }
   }
   skydive_Surfing {simSpaceSettings {
     maxAngularAcceleration=1000
     maxAngularVelocity=5
     maxLinearAcceleration=700
     maxLinearVelocity=700
     velocityScaleZ=0.5
     externalLinearDragV=<0.2 0 0>
    }}
   swimming {simSpaceSettings {
     maxAngularAcceleration=1000
     maxAngularVelocity=5
     maxLinearAcceleration=700
     maxLinearVelocity=700
     velocityScaleZ=0.5
     externalLinearDragV=<0.2 0.2 0.2>
    }}
   tactical_Sprint {simSpaceSettings {
     maxAngularAcceleration=1000
     maxAngularVelocity=3
     maxLinearAcceleration=300
     maxLinearVelocity=700
     velocityScaleZ=0
     externalLinearDragV=<0 0.5 0>
    }}
   ziplining {simSpaceSettings {
     maxAngularAcceleration=1000
     maxAngularVelocity=5
     maxLinearAcceleration=700
     maxLinearVelocity=700
     velocityScaleZ=0.5
     externalLinearDragV=<0.2 0.2 0.2>
    }}
  }
 }
 Sim Space Settings {
  type=SimSpaceSettings
  default {
   dampingAlpha=1
   maxAngularAcceleration=1000
   maxAngularVelocity=5
   maxLinearAcceleration=700
   maxLinearVelocity=700
   velocityScaleZ=0.5
   worldAlpha=1
   externalLinearDragV=<0.2 0.2 0.2>
  }
 }
]
