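# Physics asset for the face accessory part of the LimettaMech medium female player body, modelling dangling hair strands with capsules.
# One front and two side hair chains run from small base capsules, which join the head capsule through angularly free constraints.
# Links within each chain have tight swing and twist limits, mostly loosening toward the tips.
# Two collar capsules have no constraints.
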PhysicsAsset /BRCosmetics/Characters/Player/Female/Medium/Bodies/F_MED_LimettaMech/Meshes/Parts/F_MED_LimettaMech_FaceAcc_Physics
bodies {
 base_frontHair_d [{shapeType=Capsule, length=0.505, radius=0.505, shapeName=base_frontHair_d_capsule, rotation=<rot 0 -90 0>}]
 base_sidetHair_a [{shapeType=Capsule, length=0.505, radius=0.505, shapeName=base_sidetHair_a_capsule, rotation=<rot 0 -90 0>}]
 base_sidetHair_e [{shapeType=Capsule, length=0.505, radius=0.505, shapeName=base_sidetHair_e_capsule, rotation=<rot 0 -90 0>}]
 dyn_collar_l_01 [{
   shapeType=Capsule
   length=8.8855
   radius=3.466
   shapeName=dyn_collar_l_01_capsule
   center=<2.2223 0.7725 1.2231>
   rotation=<rot -90 180 100.6671>
  }]
 dyn_collar_r_01 [{
   shapeType=Capsule
   length=8.8855
   radius=3.4384
   shapeName=dyn_collar_r_01_capsule
   center=<1.9617 -1.4732 1.3867>
   rotation=<rot -90 -90 169.2754>
  }]
 dyn_frontHair_d_01 [{shapeType=Capsule, length=3, radius=1, shapeName=dyn_frontHair_d_01_capsule, center=<1 0 0>, rotation=<rot 90 0 0>}]
 dyn_frontHair_d_02 [{shapeType=Capsule, length=3, radius=1, shapeName=dyn_frontHair_d_02_capsule, center=<2 0 0>, rotation=<rot 90 0 0>}]
 dyn_frontHair_d_03 [{
   shapeType=Capsule
   length=3
   radius=0.8286
   shapeName=dyn_frontHair_d_03_capsule
   center=<1.3459 -0.3924 -0.0571>
   rotation=<rot 78.5858 8.6522 19.4562>
  }]
 dyn_sidetHair_a_01 [{shapeType=Capsule, length=1.5, radius=0.7, shapeName=dyn_sidetHair_a_01_capsule, center=<2 0 0>, rotation=<rot 90 0 0>}]
 dyn_sidetHair_a_02 [{shapeType=Capsule, length=1.5, radius=0.7, shapeName=dyn_sidetHair_a_02_capsule, center=<1.5 0 0>, rotation=<rot 90 0 0>}]
 dyn_sidetHair_a_03 [{
   shapeType=Capsule
   length=1.4982
   radius=0.5108
   shapeName=dyn_sidetHair_a_03_capsule
   center=<1.9738 -0.1931 0>
   rotation=<rot 90 -176.4237 -168.7002>
  }]
 dyn_sidetHair_e_01 [{shapeType=Capsule, length=1.5, radius=1, shapeName=dyn_sidetHair_e_01_capsule, center=<2 0 0>, rotation=<rot 90 0 0>}]
 dyn_sidetHair_e_02 [{shapeType=Capsule, length=1.5, radius=1, shapeName=dyn_sidetHair_e_02_capsule, center=<2 0 0>, rotation=<rot 90 0 0>}]
 head [{
   shapeType=Capsule
   length=7.3298
   radius=6.5347
   shapeName=head_capsule
   center=<3.3412 0.6229 0.4238>
   rotation=<rot -1.6759 -90.0001 85.5681>
  }]
}
constraints [
 {bone1Name=dyn_frontHair_d_01, bone2Name=base_frontHair_d, swing1LimitDegrees=2, swing2LimitDegrees=5, twistLimitDegrees=1}
 {bone1Name=dyn_frontHair_d_02, bone2Name=dyn_frontHair_d_01, swing1LimitDegrees=3, swing2LimitDegrees=7, twistLimitDegrees=1}
 {bone1Name=dyn_frontHair_d_03, bone2Name=dyn_frontHair_d_02, swing1LimitDegrees=5, swing2LimitDegrees=10, twistLimitDegrees=1}
 {bone1Name=dyn_sidetHair_a_01, bone2Name=base_sidetHair_a, swing1LimitDegrees=3, swing2LimitDegrees=10, twistLimitDegrees=1}
 {bone1Name=dyn_sidetHair_a_02, bone2Name=dyn_sidetHair_a_01, swing1LimitDegrees=5, swing2LimitDegrees=12, twistLimitDegrees=1}
 {bone1Name=dyn_sidetHair_a_03, bone2Name=dyn_sidetHair_a_02, swing1LimitDegrees=6, swing2LimitDegrees=15, twistLimitDegrees=1}
 {bone1Name=dyn_sidetHair_e_01, bone2Name=base_sidetHair_e, swing1LimitDegrees=5, swing2LimitDegrees=2, twistLimitDegrees=1}
 {bone1Name=dyn_sidetHair_e_02, bone2Name=dyn_sidetHair_e_01, swing1LimitDegrees=5, swing2LimitDegrees=2, twistLimitDegrees=1}
 {
  bone1Name=base_sidetHair_e
  bone2Name=head
  swing1LimitDegrees=45
  swing1Motion=Free
  swing2LimitDegrees=45
  swing2Motion=Free
  twistLimitDegrees=45
  twistMotion=Free
 }
 {
  bone1Name=base_sidetHair_a
  bone2Name=head
  swing1LimitDegrees=45
  swing1Motion=Free
  swing2LimitDegrees=45
  swing2Motion=Free
  twistLimitDegrees=45
  twistMotion=Free
 }
 {
  bone1Name=base_frontHair_d
  bone2Name=head
  swing1LimitDegrees=45
  swing1Motion=Free
  swing2LimitDegrees=45
  swing2Motion=Free
  twistLimitDegrees=45
  twistMotion=Free
 }
]
