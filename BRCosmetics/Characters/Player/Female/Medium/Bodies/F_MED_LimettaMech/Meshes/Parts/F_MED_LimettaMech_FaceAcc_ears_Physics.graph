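# Physics asset for the bunny ear face accessory on the LimettaMech female character, setting up simulated secondary motion for both ears.
# Each ear is a chain of four capsule bodies on dynamic bones, hanging from a small capsule on its base bone.
# Constraints link each segment to the one before it with tight swing and twist limits, so the ears can flex only a few degrees per joint.
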PhysicsAsset /BRCosmetics/Characters/Player/Female/Medium/Bodies/F_MED_LimettaMech/Meshes/Parts/F_MED_LimettaMech_FaceAcc_ears_Physics
bodies {
 base_bunnyHear_l [{shapeType=Capsule, length=0.505, radius=0.505, shapeName=base_bunnyHear_l_capsule, rotation=<rot 0 -90 0>}]
 base_bunnyHear_r [{shapeType=Capsule, length=0.505, radius=0.505, shapeName=base_bunnyHear_r_capsule, rotation=<rot 0 -90 0>}]
 dyn_bunnyHear_l_01 [{shapeType=Capsule, length=1, radius=1, shapeName=dyn_bunnyHear_l_01_capsule, center=<1 0 0>, rotation=<rot 90 0 0>}]
 dyn_bunnyHear_l_02 [{shapeType=Capsule, length=1, radius=1, shapeName=dyn_bunnyHear_l_02_capsule, center=<1 0 0>, rotation=<rot 90 0 0>}]
 dyn_bunnyHear_l_03 [{shapeType=Capsule, length=1, radius=1, shapeName=dyn_bunnyHear_l_03_capsule, center=<1 0 0>, rotation=<rot 90 0 0>}]
 dyn_bunnyHear_l_04 [{shapeType=Capsule, length=1, radius=1, shapeName=dyn_bunnyHear_l_04_capsule, center=<1 0 0>, rotation=<rot 90 0 0>}]
 dyn_bunnyHear_r_01 [{shapeType=Capsule, length=1, radius=1, shapeName=dyn_bunnyHear_r_01_capsule, center=<1 0 0>, rotation=<rot 90 0 0>}]
 dyn_bunnyHear_r_02 [{shapeType=Capsule, length=1, radius=1, shapeName=dyn_bunnyHear_r_02_capsule, center=<1 0 0>, rotation=<rot 90 0 0>}]
 dyn_bunnyHear_r_03 [{shapeType=Capsule, length=1, radius=1, shapeName=dyn_bunnyHear_r_03_capsule, center=<1 0 0>, rotation=<rot 90 0 0>}]
 dyn_bunnyHear_r_04 [{shapeType=Capsule, length=1, radius=1, shapeName=dyn_bunnyHear_r_04_capsule, center=<1 0 0>, rotation=<rot 90 0 0>}]
}
constraints [
 {bone1Name=dyn_bunnyHear_l_01, bone2Name=base_bunnyHear_l, swing1LimitDegrees=5, swing2LimitDegrees=4, twistLimitDegrees=2}
 {bone1Name=dyn_bunnyHear_l_02, bone2Name=dyn_bunnyHear_l_01, swing1LimitDegrees=5, swing2LimitDegrees=4, twistLimitDegrees=2}
 {bone1Name=dyn_bunnyHear_l_03, bone2Name=dyn_bunnyHear_l_02, swing1LimitDegrees=5, swing2LimitDegrees=4, twistLimitDegrees=2}
 {bone1Name=dyn_bunnyHear_l_04, bone2Name=dyn_bunnyHear_l_03, swing1LimitDegrees=5, swing2LimitDegrees=4, twistLimitDegrees=2}
 {bone1Name=dyn_bunnyHear_r_01, bone2Name=base_bunnyHear_r, swing1LimitDegrees=5, swing2LimitDegrees=4, twistLimitDegrees=2}
 {bone1Name=dyn_bunnyHear_r_02, bone2Name=dyn_bunnyHear_r_01, swing1LimitDegrees=5, swing2LimitDegrees=4, twistLimitDegrees=2}
 {bone1Name=dyn_bunnyHear_r_03, bone2Name=dyn_bunnyHear_r_02, swing1LimitDegrees=5, swing2LimitDegrees=4, twistLimitDegrees=2}
 {bone1Name=dyn_bunnyHear_r_04, bone2Name=dyn_bunnyHear_r_03, swing1LimitDegrees=5, swing2LimitDegrees=4, twistLimitDegrees=2}
]
